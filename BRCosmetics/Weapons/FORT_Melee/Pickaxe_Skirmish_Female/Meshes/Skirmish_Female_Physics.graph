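# Physics asset for the Skirmish Female two-handed pickaxe, modeling the weapon and its two dangling cloth strips as rigid bodies.
# Two capsules on the handle bone cover the weapon, and a box on dyn_cloth anchors two five-link chains of mostly thin boxes, a and b.
# Swing-limited constraints link each chain, with looser twist at the top two joints.
# No constraint joins dyn_cloth to the handle.
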PhysicsAsset /BRCosmetics/Weapons/FORT_Melee/Pickaxe_Skirmish_Female/Meshes/Skirmish_Female_Physics
bodies {
 Melee_TwoHanded_Handle [
  {shapeType=Capsule, length=65.5685, radius=2.5199, center=<-0.3423 0.1461 13.3597>, rotation=<rot -0.1914 -90 1.7646>}
  {shapeType=Capsule, length=30.1859, radius=5.0675, center=<-1.8346 18.6425 -2.3511>, rotation=<rot 0 0 -90.5229>}
 ]
 dyn_cloth [{shapeType=Box, extentX=7.5, extentY=7.5, extentZ=8.3006, center=<0 0 -1.0132>}]
 dyn_cloth_a00 [{shapeType=Box, extentX=5.1548, extentY=1.0733, extentZ=5.9624, center=<0 0.3107 -2.6335>, rotation=<rot 0 0 7.2152>}]
 dyn_cloth_a01 [{shapeType=Box, extentX=4.3587, extentY=1.1556, extentZ=6.6767, center=<-0.0001 0 -3.3278>}]
 dyn_cloth_a02 [{shapeType=Box, extentX=4.3587, extentY=1.1556, extentZ=7.842, center=<-0.0001 0 -2.3724>}]
 dyn_cloth_a03 [{shapeType=Box, extentX=4.3587, extentY=1.1556, extentZ=7.5332, center=<0 0 -1.602>}]
 dyn_cloth_a04 [{shapeType=Box, extentX=4.3587, extentY=1.1556, extentZ=9.0395, center=<0 0 -1.3985>}]
 dyn_cloth_b00 [{shapeType=Box, extentX=5.1548, extentY=1.0733, extentZ=6.2561, center=<0 0.0872 -2.8835>, rotation=<rot 0 0 7.766>}]
 dyn_cloth_b01 [{shapeType=Box, extentX=4.3587, extentY=1.1556, extentZ=6.6767, center=<-0.0001 0 -3.3278>}]
 dyn_cloth_b02 [{shapeType=Box, extentX=4.3587, extentY=1.1556, extentZ=7.842, center=<-0.0001 0 -2.3724>}]
 dyn_cloth_b03 [{shapeType=Box, extentX=4.3587, extentY=1.1556, extentZ=7.5332, center=<0 0 -1.602>}]
 dyn_cloth_b04 [{shapeType=Box, extentX=4.3587, extentY=1.1556, extentZ=9.0395, center=<0 0 -1.3985>}]
}
constraints [
 {bone1Name=dyn_cloth_a01, bone2Name=dyn_cloth_a00, swing1LimitDegrees=45, swing2LimitDegrees=45, twistLimitDegrees=90}
 {bone1Name=dyn_cloth_a02, bone2Name=dyn_cloth_a01, swing1LimitDegrees=45, swing2LimitDegrees=45, twistLimitDegrees=45}
 {bone1Name=dyn_cloth_a03, bone2Name=dyn_cloth_a02, swing1LimitDegrees=45, swing2LimitDegrees=45, twistLimitDegrees=45}
 {bone1Name=dyn_cloth_a04, bone2Name=dyn_cloth_a03, swing1LimitDegrees=45, swing2LimitDegrees=45, twistLimitDegrees=45}
 {bone1Name=dyn_cloth_b01, bone2Name=dyn_cloth_b00, swing1LimitDegrees=45, swing2LimitDegrees=45, twistLimitDegrees=90}
 {bone1Name=dyn_cloth_b02, bone2Name=dyn_cloth_b01, swing1LimitDegrees=45, swing2LimitDegrees=45, twistLimitDegrees=45}
 {bone1Name=dyn_cloth_b03, bone2Name=dyn_cloth_b02, swing1LimitDegrees=45, swing2LimitDegrees=45, twistLimitDegrees=45}
 {bone1Name=dyn_cloth_b04, bone2Name=dyn_cloth_b03, swing1LimitDegrees=45, swing2LimitDegrees=45, twistLimitDegrees=45}
 {bone1Name=dyn_cloth_b00, bone2Name=dyn_cloth, swing1LimitDegrees=45, swing2LimitDegrees=45, twistLimitDegrees=90}
 {bone1Name=dyn_cloth_a00, bone2Name=dyn_cloth, swing1LimitDegrees=45, swing2LimitDegrees=45, twistLimitDegrees=90}
]
